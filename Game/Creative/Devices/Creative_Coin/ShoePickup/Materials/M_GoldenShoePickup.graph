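# Material for the golden shoe pickup in the Creative coin device, rendering a shiny, tintable metal surface.
# A color parameter, full metallic, a roughness parameter and the shoe normal map feed a Material Attributes set.
# The shared coin bobble function supplies World Position Offset, animating the pickup in the shader.
# The attributes then pass through the Creative move tool function, and the material supports skeletal meshes and mesh deformers.
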Material /Game/Creative/Devices/Creative_Coin/ShoePickup/Materials/M_GoldenShoePickup
graph {
 properties {
  bUseMaterialAttributes=true
  bUsedWithMeshDeformer=true
  bUsedWithSkeletalMesh=true
  pixelDepthOffsetMode=PDOM_Legacy
  usedShadingModels=Default Lit
 }
 nodes [
  n1 = TextureSampleParameter2D() {parameterName=Normals, samplerType=SAMPLERTYPE_Normal, texture=/Game/Creative/Devices/Creative_Coin/ShoePickup/Textures/T_ShoePickup_N}
  n2 = SetMaterialAttributes!(
   MaterialAttributes
   Base Color=VectorParameter() {parameterName=Color, defaultValue=<rgba 1 0.5683 0.2174 1>}.RGB
   Metallic=Constant() {r=1}
   Roughness=ScalarParameter() {defaultValue=0.2419, parameterName=Shine Roughnmess}
   Normal=@n1.RGB
   World Position Offset=MaterialFunctionCall!(TimeScale, MaxHeight, MinHeight) {materialFunction=/Game/Playgrounds/Items/Props/Materials/MF_Creative_CoinBobble}.WorldPositionOffset
  ) {attributeSetTypes [
    #1
    #2
    #3
    #4
    #5
   ]}
  MaterialFunctionCall(MA In=@n2) {materialFunction=&MF_CreativeMoveTool}
 ]
}
